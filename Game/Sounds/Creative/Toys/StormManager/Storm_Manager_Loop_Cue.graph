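# Looping sound cue for the Storm Manager Creative toy, routed through the Creative gadgets sound class at reduced volume.
# A single looping wave player feeds the output with one storm loop wave.
# Overridden attenuation gives logarithmic distance falloff with low-pass filtering, stereo spread and occlusion traced on a game trace channel.
# Interpolated listener focus lowers the volume when the source is outside the focus angle.
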SoundCue /Game/Sounds/Creative/Toys/StormManager/Storm_Manager_Loop_Cue
root=WavePlayer_0
nodes [{id=WavePlayer_0, properties {bLooping=true, soundWaveAssetPtr=&/Storm_Manager_Loop_01}}]
properties {
 bOverrideAttenuation=true
 maxDistance=1050
 soundClassObject=&Fort_Gadgets_Creative
 volumeMultiplier=0.5
 attenuationOverrides {
  bAttenuateWithLPF=true
  bEnableFocusInterpolation=true
  bEnableListenerFocus=true
  bEnableOcclusion=true
  distanceAlgorithm=Logarithmic
  falloffDistance=800
  focusAzimuth=15
  lPFFrequencyAtMax=200
  lPFRadiusMax=1200
  lPFRadiusMin=400
  nonFocusAzimuth=45
  nonFocusVolumeAttenuation=0.5
  occlusionTraceChannel=ECC_GameTraceChannel11
  stereoSpread=500
  attenuationShapeExtents {x=250}
 }
}
